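# Material function for Vault World front-end item preview transitions that outputs floor color, metallic, specular, roughness and emissive color.
# It saturates the input Vault World color and builds a linear radial gradient mask from centered UVs, sharing both through named reroutes.
# Nested lerps shape the surface outputs, with the saturated IsFloor scalar parameter gating the floor look.
# A shading path switch gives mobile a constant roughness in place of the default path's blended value.
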MaterialFunction /Game/Environments/FrontEnd/Blueprints/ItemPreview/VaultWorldTransitions/MaterialFunctions/MF_UI_VaultWorld_FloorSetup
graph {
 properties {libraryCategoriesText [Vault World]}
 nodes [
  n1 = NamedRerouteDeclaration(None=Saturate(None=FunctionInput() {bUsePreviewValueAsDefault=true, inputName=Vault World Color, previewValue=<0.1527 1 0 1>})) {name=Final Color, nodeColor=<rgba 0 0 0 1>}
  n2 = NamedRerouteUsage()
  n3 = Saturate(None=ScalarParameter() {group=Floor, parameterName=IsFloor, sliderMax=1, sortPriority=-1})
  FunctionOutput(None=LinearInterpolate(B=LinearInterpolate(B=@n1, Alpha=@n2), Alpha=NamedRerouteDeclaration(None=@n3) {name=Is Floor, nodeColor=<rgba 0 0 0 1>})) {outputName=Floor Color}
  n4 = Constant() {r=1}
  n5 = NamedRerouteUsage()
  FunctionOutput(None=LinearInterpolate(B=LinearInterpolate(B=@n4, Alpha=@n5) {constA=1}, Alpha=NamedRerouteUsage())) {outputName=Floor Metallic, sortPriority=1}
  n6 = Constant() {r=2}
  n7 = NamedRerouteUsage()
  FunctionOutput(None=LinearInterpolate(B=LinearInterpolate(B=@n6, Alpha=@n7), Alpha=NamedRerouteUsage())) {outputName=Floor Specular, sortPriority=2}
  n8 = LinearInterpolate(B=Constant() {r=0.2}, Alpha=NamedRerouteUsage()) {constA=1}
  n9 = Constant() {r=0.75}
  FunctionOutput(None=LinearInterpolate(B=ShadingPathSwitch(Default=@n8, Mobile=@n9), Alpha=NamedRerouteUsage()) {constA=1}) {outputName=Floor Roughness, sortPriority=3}
  FunctionOutput(None=NamedRerouteUsage()) {outputName=Floor Emissive Color, sortPriority=4}
  n10 = Length(None=ConstantBiasScale(None=TextureCoordinate()) {bias=-0.5, scale=2}) {is=class MaterialExpressionLength}
  NamedRerouteDeclaration(None=Saturate(None=OneMinus(None=@n10)) {desc=Linear Radial Gradient}) {name=Floor Gradient Mask, nodeColor=<rgba 0 0 0 1>}
 ]
}
