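# Material function for front-end Vault World UI transitions that centralizes UV coordinates, switching between screen space and local image-plane space.
# A Use Local UV scalar, defaulting to a parameter of that name, lerps every output between the viewport UV and texture coordinate paths.
# Two UI scaling functions take texture size and pivot, providing fit and fill UVs plus X/Y scales alongside the raw UV.
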MaterialFunction /Game/Environments/FrontEnd/Blueprints/ItemPreview/VaultWorldTransitions/MaterialFunctions/MF_UI_VaultWorld_UV
graph {
 properties {bExposeToLibrary=true, description=Use to centralize UV coordinates, libraryCategoriesText [Vault World]}
 nodes [
  n1 = ScreenPosition()
  n2 = TextureCoordinate()
  n3 = FunctionInput(Preview=ScalarParameter() {group=UV, parameterName=UseLocalUV, sliderMax=1}) {
   bUsePreviewValueAsDefault=true
   description=Use local UV if you want your coordinates to be mapped to the image plane instead of the screen.
   inputName=Use Local UV
   inputType=FunctionInput_Scalar
  }
  FunctionOutput(None=LinearInterpolate(A=@n1.ViewportUV, B=@n2, Alpha=@n3)) {outputName=UV}
  n4 = FunctionInput() {
   bUsePreviewValueAsDefault=true
   description=Input the texture size
   inputName=Texture Size
   inputType=FunctionInput_Vector2
   sortPriority=1
   previewValue=<1 1 0 1>
  }
  n5 = FunctionInput() {
   bUsePreviewValueAsDefault=true
   description=Scale from Center by default. Use to define where the pivot point influence the scale
   inputName=Scale Pivot Point
   inputType=FunctionInput_Vector2
   sortPriority=5
   previewValue=<0.5 0.5 0 1>
  }
  n6 = MaterialFunctionCall!(ScreenUV=@n1.ViewportUV, TextureSize=@n4, ScalePivotPoint=@n5) {materialFunction=/FortUILibrary/MaterialFunctions/Basics/MF_UI_ScreenSpaceUVScaling}
  n7 = MaterialFunctionCall(UV=@n2, TextureSize=@n4, ScalePivotPoint=@n5) {materialFunction=&MF_UI_UVScaling}
  FunctionOutput(None=LinearInterpolate(A=@n6.Scaled to Fit, B=@n7.Scaled to Fit, Alpha=@n3)) {outputName=Scaled to Fit, sortPriority=5}
  FunctionOutput(None=LinearInterpolate(A=@n6.Scaled to Fill, B=@n7.Scaled To Fill, Alpha=@n3)) {outputName=Scaled to Fill, sortPriority=10}
  FunctionOutput(None=LinearInterpolate(A=@n6.ScaleX, B=@n7.ScaleX, Alpha=@n3)) {outputName=Scale X, sortPriority=15}
  FunctionOutput(None=LinearInterpolate(A=@n6.ScaleY, B=@n7.ScaleY, Alpha=@n3)) {outputName=Scale Y, sortPriority=20}
 ]
}
