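# Material function that adds a soft radial focal point to the front end's vault world backgrounds.
# Positions it from the shared VaultWorld parameter collection or override parameters, outputting its UV and the collection's alpha as a transition value.
# Builds a scalable, sine-eased radial mask, multiplied by a blend input for custom effects.
# Derives a brightened focal color according to luminosity, replaceable by an override input, and outputs the blended color.
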MaterialFunction /Game/Environments/FrontEnd/Blueprints/ItemPreview/VaultWorldTransitions/MaterialFunctions/MF_UI_VaultWorld_FocalPoint
graph {
 properties {bExposeToLibrary=true, description=Use to incorporate a focal point in the vault world background, libraryCategoriesText [Vault World]}
 nodes [
  n1 = NamedRerouteUsage()
  n2 = NamedRerouteUsage()
  n3 = Constant() {desc=Luminosity, r=0.9}
  n4 = Max(A=OneMinus(None=NamedRerouteDeclaration(None=@n3) {name=Luminosity, nodeColor=<rgba 0 0 0 1>})) {constB=0}
  n5 = Desaturation(None=NamedRerouteUsage()) {desc=Push toward brighter values depending on color luminosity}
  n6 = LinearInterpolate(A=Saturate(None=Divide(A=@n2, B=@n4)), B=LinearInterpolate(A=NamedRerouteUsage(), Alpha=NamedRerouteUsage()), Alpha=@n5)
  n7 = FunctionInput(Preview=@n6) {
   bUsePreviewValueAsDefault=true
   description=Allow you to inject a specific color instead of relying on the automatic color variation.
   inputName=Focal Color Override
   sortPriority=5
  }
  n8 = Multiply(A=NamedRerouteUsage(), B=Constant() {desc=Focal Opacity, r=1})
  FunctionOutput(None=Saturate(None=LinearInterpolate(A=@n1, B=@n7, Alpha=@n8))) {description=Provide the color including the focal point, outputName=Focal Color}
  n9 = MaterialFunctionCall!(Use Local UV, Texture Size, Scale Pivot Point) {materialFunction=&/MF_UI_VaultWorld_UV}
  n10 = ComponentMask(None=@n9.Scale Y) {r=true}
  n11 = CollectionParameter() {
   collection=/Game/Environments/FrontEnd/Blueprints/ItemPreview/VaultWorldTransitions/MaterialCollectionParameters/MPC_UI_Master_VaultWorld_Settings
   parameterName=FocalPointPosition
  }
  n12 = LinearInterpolate(
   A=ComponentMask(None=@n11) {g=true, r=true}
   B=AppendVector(
    A=ScalarParameter() {group=Focal Point, parameterName=FocalPositionX, sortPriority=5}
    B=ScalarParameter() {group=Focal Point, parameterName=FocalPositionY, sortPriority=10}
   )
   Alpha=ScalarParameter() {group=Focal Point, parameterName=Override Focal Position, sliderMax=1}
  )
  n13 = ComponentMask(None=@n12) {r=true}
  n14 = ComponentMask(None=@n9.Scale Y) {g=true}
  n15 = ComponentMask(None=@n12) {g=true}
  n16 = FunctionOutput(None=AppendVector(A=Subtract(A=@n10, B=@n13), B=Add(A=@n14, B=@n15))) {
   description=Ouput the UV of the focal point including its position if you need to create your own focal visual.
   outputName=Focal Point UV
   sortPriority=2
  }
  n17 = FunctionInput() {
   bUsePreviewValueAsDefault=true
   description=Define focal point scale
   inputName=Focal Scale Override
   inputType=FunctionInput_Vector2
   sortPriority=30
   previewValue=<1 1 0 1>
  }
  n18 = Length(None=ConstantBiasScale(None=MaterialFunctionCall(UVs=@n16, Texture Scale=@n17) {materialFunction=&ScaleUVsByCenter}.UVs) {bias=-0.5, scale=2}) {is=class MaterialExpressionLength}
  n19 = MaterialFunctionCall("Value (0-1)"=Saturate(None=OneMinus(None=@n18))) {materialFunction=&MF_UI_EaseCurves}
  n20 = FunctionInput() {
   bUsePreviewValueAsDefault=true
   description=Use to blend the focal with a custom effect
   inputName=Mask Blend
   inputType=FunctionInput_Scalar
   sortPriority=50
   previewValue=<1 0 0 1>
  }
  FunctionOutput(None=NamedRerouteDeclaration(None=Multiply(A=@n19.EaseInSine, B=@n20)) {name=Focal Mask, nodeColor=<rgba 0 0 0 1>}) {description=Ouput the mask of the focal, outputName=Focal Mask, sortPriority=1}
  n21 = CollectionParameter() {
   collection=&MPC_UI_Master_VaultWorld_Settings
   parameterName=FocalPointPosition
  }
  FunctionOutput(None=ComponentMask(None=@n21) {a=true}) {outputName=Transition Value, sortPriority=3}
  LinearInterpolate(A=@n6, B=LinearInterpolate(A=NamedRerouteUsage(), Alpha=@n5), Alpha=NamedRerouteUsage())
  n22 = FunctionInput() {bUsePreviewValueAsDefault=true, description=Set color of the vault world, inputName=Vault World Color, previewValue=<0.2425 0.25 0.0018 1>}
  NamedRerouteDeclaration(None=@n22) {name=Vault World Color, nodeColor=<rgba 0 0 0 1>}
  n23 = NamedRerouteUsage()
  n24 = ComponentMask(None=@n23) {r=true}
  n25 = ComponentMask(None=@n23) {g=true}
  n26 = ComponentMask(None=@n23) {b=true}
  NamedRerouteDeclaration(None=Min(
    A=MaterialFunctionCall!(A=@n24, B=@n25) {materialFunction=/FortUILibrary/MaterialFunctions/Basics/MF_UI_BooleanOperations}.IsEqual
    B=MaterialFunctionCall(A=@n25, B=@n26) {materialFunction=&MF_UI_BooleanOperations}.IsEqual
   )) {name=Is Grayscale, nodeColor=<rgba 0 0 0 1>}
 ]
}
