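# Material function for front-end vault world materials that calibrates the final color to counter the vault world's post process.
# Applies a Photoshop-style levels adjustment: the input color is remapped from a constant min/max range and clamped to 0-1.
# A gamma correction then raises the result to the inverse of a constant gamma and saturates it as Calibrated Color.
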MaterialFunction /Game/Environments/FrontEnd/Blueprints/ItemPreview/VaultWorldTransitions/MaterialFunctions/MF_UI_VaultWorld_PostProcessCalibration
graph {
 properties {
  bExposeToLibrary=true
  description=Use this function to counter the post process that is occurring in the vault world
  libraryCategoriesText [Vault World]
 }
 nodes [
  n1 = FunctionInput() {bUsePreviewValueAsDefault=true, description=Pass the final vault world color, inputName=Vault World Color, previewValue=<1 0.0229 0 1>}
  n2 = Constant() {desc=Min}
  n3 = Subtract(A=@n1, B=@n2)
  n4 = Constant()
  n5 = Constant() {desc=Max, r=1.8}
  n6 = Min(A=Divide(A=Max(A=@n3, B=@n4), B=Subtract(A=@n5, B=@n2)), B=Constant() {r=1}) {desc=LevelsControlInputRange}
  n7 = Divide(A=Constant() {r=1}, B=Constant() {desc=Gamma, r=1.2})
  FunctionOutput(None=Saturate(None=Power(Base=@n6, Exp=@n7) {desc=LevelsControlInput})) {description=Output a calibrated color to counter the post process, outputName=Calibrated Color}
 ]
}
