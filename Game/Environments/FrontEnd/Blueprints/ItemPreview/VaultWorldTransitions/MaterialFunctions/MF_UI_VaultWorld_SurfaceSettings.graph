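# Material function for the front-end Vault World item preview that derives base color, metallic, specular, roughness and emissive from an input color.
# Floor parameters build a soft radial mask from offset, rotated and scaled UVs, alongside IsFloor and reveal-transition scalars.
# These named reroutes drive lerps between inner/outer or floor values for each output.
# Quality and shading-path switches change mask scale and opacity and zero floor specular on Low and Mobile.
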MaterialFunction /Game/Environments/FrontEnd/Blueprints/ItemPreview/VaultWorldTransitions/MaterialFunctions/MF_UI_VaultWorld_SurfaceSettings
graph {
 properties {libraryCategoriesText [Vault World]}
 nodes [
  n1 = NamedRerouteUsage()
  n2 = NamedRerouteUsage()
  FunctionOutput(None=LinearInterpolate(B=NamedRerouteUsage(), Alpha=Multiply(A=@n1, B=@n2))) {outputName=Base Color}
  FunctionOutput(None=LinearInterpolate(A=Constant() {r=1}, B=ScalarParameter() {defaultValue=0.9, group=Floor, parameterName=Metallic, sliderMax=1}, Alpha=NamedRerouteUsage())) {outputName=Metallic, sortPriority=1}
  n3 = Constant()
  n4 = QualitySwitch(
   Default=LinearInterpolate(
    A=ScalarParameter() {desc=Specular, group=Floor, parameterName=Specular_Outer, sliderMax=1}
    B=ScalarParameter() {defaultValue=0.45, desc=Specular, group=Floor, parameterName=Specular_Inner, sliderMax=1}
    Alpha=NamedRerouteUsage()
   ) {constA=0.25, constB=0.001}
   Low=@n3
  )
  FunctionOutput(None=LinearInterpolate(B=ShadingPathSwitch(Default=@n4, Mobile=@n3), Alpha=NamedRerouteUsage())) {outputName=Specular, sortPriority=2}
  n5 = ScalarParameter() {defaultValue=1, group=Floor, parameterName=Roughness_Outer, sliderMax=1, sliderMin=0.001}
  n6 = ScalarParameter() {defaultValue=0.01, group=Floor, parameterName=Roughness_Inner, sliderMax=1, sliderMin=0.001}
  n7 = NamedRerouteUsage()
  FunctionOutput(None=LinearInterpolate(B=LinearInterpolate(A=@n5, B=@n6, Alpha=@n7) {constA=0.25, constB=0.001}, Alpha=NamedRerouteUsage()) {constA=1, constB=0.001}) {outputName=Roughness, sortPriority=3}
  n8 = NamedRerouteUsage()
  n9 = ScalarParameter() {defaultValue=1, group=Floor, parameterName=Emissive_Inner_Intensity, sliderMax=1}
  n10 = NamedRerouteUsage()
  n11 = NamedRerouteUsage()
  FunctionOutput(None=LinearInterpolate(A=@n8, B=Multiply(A=@n8, B=@n9) {constB=0.6}, Alpha=LinearInterpolate(B=@n10, Alpha=@n11)) {constB=0}) {outputName=Emissive Color, sortPriority=4}
  NamedRerouteDeclaration(None=Saturate(None=FunctionInput() {bUsePreviewValueAsDefault=true, inputName=Vault World Color, previewValue=<0.1527 1 0 1>})) {name=Final Color, nodeColor=<rgba 0 0 0 1>}
  n12 = ScalarParameter() {group=Floor, parameterName=RadialFloorOffsetX}
  n13 = ScalarParameter() {group=Floor, parameterName=RadialFloorOffsetY}
  n14 = Rotator(
   Coordinate=Add(A=TextureCoordinate(), B=AppendVector(A=@n12, B=@n13))
   Time=ScalarParameter() {group=Floor, parameterName=RadialFloorRotation}
  ) {speed=6.2832}
  n15 = AppendVector(
   A=ScalarParameter() {defaultValue=0.5, group=Floor, parameterName=RadialFloorScaleX_LowQuality, sliderMax=1}
   B=ScalarParameter() {defaultValue=0.5, group=Floor, parameterName=RadialFloorScaleY_LowQuality, sliderMax=1}
  )
  n16 = QualitySwitch(
   Default=AppendVector(
    A=ScalarParameter() {defaultValue=0.5, group=Floor, parameterName=RadialFloorScaleX, sliderMax=1}
    B=ScalarParameter() {defaultValue=0.5, group=Floor, parameterName=RadialFloorScaleY, sliderMax=1}
   )
   Low=@n15
  )
  n17 = Length(None=Divide(A=Subtract(A=@n14) {constB=0.5}, B=ShadingPathSwitch(Default=@n16, Mobile=@n15) {desc=Change scale based on platform/graphic settings})) {is=class MaterialExpressionLength}
  n18 = SmoothStep(Value=Saturate(None=OneMinus(None=@n17)))
  n19 = Multiply(A=@n18, B=ScalarParameter() {defaultValue=1, group=Floor, parameterName=RadialFloorOpacity_LowQuality, sliderMax=1})
  NamedRerouteDeclaration(None=ShadingPathSwitch(Default=QualitySwitch(Default=@n18, Low=@n19), Mobile=@n19) {desc=Change opacity based on platform/graphic settings}) {name=Floor Mask, nodeColor=<rgba 0 0 0 1>}
  NamedRerouteDeclaration(None=ScalarParameter() {group=Floor, parameterName=IsFloor, sliderMax=1, sortPriority=-1}) {name=Is Floor, nodeColor=<rgba 0 0 0 1>}
  NamedRerouteDeclaration(None=Saturate(None=ScalarParameter() {defaultValue=1, group=Floor, parameterName=Floor Reveal Transition, sliderMax=1})) {name=Floor Reveal, nodeColor=<rgba 0 0 0 1>}
 ]
}
