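# Final-stage material function for front-end vault world materials that drives transitions between their visuals and outputs surface attributes plus an opacity mask.
# MPC direction, loading and reveal values drive hard-edged masks, built with Ceil thresholds on rotated, screen-fitted UV gradients with an Invert parameter.
# Input color is lerped with a nebulous loading background, eased toward dark navy, then passed through SurfaceSettings.
# A FloorSetup call also receives that color, but its result feeds no output.
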MaterialFunction /Game/Environments/FrontEnd/Blueprints/ItemPreview/VaultWorldTransitions/MaterialFunctions/MF_UI_VaultWorld_Transition
graph {
 properties {
  bExposeToLibrary=true
  description=It is the bridge between your visuals and the material output. It needs to be used at the end completely. It will control the transitions between different materials and display the floor in the vault world
  libraryCategoriesText [Vault World]
 }
 nodes [
  FunctionOutput(None=NamedRerouteUsage()) {outputName=Opacity Mask, sortPriority=50}
  n1 = FunctionInput() {bUsePreviewValueAsDefault=true, inputName=Vault World Color, previewValue=<1 0 0.7676 1>}
  n2 = MaterialFunctionCall!(
   Primary Color=Constant3Vector() {constant=<rgba 0 0.3274 0.9823 1>}
   Secondary Color=Constant3Vector() {constant=<rgba 0 0.2423 0.9734 1>}
   Background Color
   Wind Opacity=Constant() {r=0.3}
   Rotation Multiplier=NamedRerouteUsage()
   Triangle Opacity=Constant() {r=2}
   Offset Multiplier=Multiply(A=NamedRerouteUsage()) {constB=0.025}
  ) {materialFunction=/FortUILibrary/Materials/Background/Nebulous/MF_UI_BackgroundNebulous}
  n3 = NamedRerouteDeclaration(None=LinearInterpolate(
    A=NamedRerouteDeclaration(None=@n1) {name=Vault World Color, nodeColor=<rgba 0 0 0 1>}
    B=NamedRerouteDeclaration(None=@n2.Final Color) {name=Loading Color, nodeColor=<rgba 0 0 0 1>}
    Alpha=NamedRerouteUsage()
   )) {name=Color, nodeColor=<rgba 0 0 0 1>}
  n4 = Power(Base=@n3) {constExponent=2.2}
  n5 = MaterialFunctionCall("Value (0-1)"=Saturate(None=NamedRerouteUsage())) {materialFunction=&MF_UI_EaseCurves}
  n6 = Saturate(None=LinearInterpolate(
    A=LinearInterpolate(A=@n3, B=@n4, Alpha=@n5.EaseInSine)
    B=Constant3Vector() {constant=<rgba 0.0108 0.0088 0.0677 1>}
    Alpha=MaterialFunctionCall("Value (0-1)"=@n5.EaseInCubic) {materialFunction=&MF_UI_EaseCurves}.EaseInQuint
   ))
  n7 = MaterialFunctionCall!(Vault World Color=@n6) {materialFunction=&/MF_UI_VaultWorld_SurfaceSettings}
  FunctionOutput(None=@n7.Base Color) {outputName=Color}
  FunctionOutput(None=@n7.Emissive Color) {outputName=Emissive Color, sortPriority=20}
  FunctionOutput(None=@n7.Metallic) {outputName=Metallic, sortPriority=5}
  FunctionOutput(None=@n7.Specular) {outputName=Specular, sortPriority=10}
  FunctionOutput(None=@n7.Roughness) {outputName=Roughness, sortPriority=15}
  MaterialFunctionCall!(Vault World Color=@n6) {materialFunction=&/MF_UI_VaultWorld_FloorSetup}
  n8 = CollectionParameter() {
   collection=/Game/Environments/FrontEnd/Blueprints/ItemPreview/VaultWorldTransitions/MaterialCollectionParameters/MPC_UI_Master_VaultWorld_Settings
   parameterName=Transition Direction
  }
  n9 = Rotator(
   Coordinate=MaterialFunctionCall!(ScreenUV=ScreenPosition().ViewportUV, TextureSize, ScalePivotPoint) {materialFunction=/FortUILibrary/MaterialFunctions/Basics/MF_UI_ScreenSpaceUVScaling}.Scaled to Fit
   Time=NamedRerouteDeclaration(None=Add(A=@n8) {constB=0.0075, desc=Additional Rotation}) {name=Rotation Angle, nodeColor=<rgba 0 0 0 1>}
  ) {speed=6.2832}
  NamedRerouteDeclaration(None=@n9) {name=Common UV, nodeColor=<rgba 0 0 0 1>}
  n10 = NamedRerouteUsage()
  n11 = NamedRerouteDeclaration(None=ComponentMask(None=@n10) {g=true}) {name=Width Scale, nodeColor=<rgba 0 0 0 1>}
  n12 = NamedRerouteUsage()
  n13 = Abs(None=Cosine(None=@n12))
  n14 = Abs(None=Sine(None=@n12))
  n15 = NamedRerouteDeclaration(None=ComponentMask(None=@n10) {r=true}) {name=Height Scale, nodeColor=<rgba 0 0 0 1>}
  NamedRerouteDeclaration(None=Add(A=Multiply(A=@n11, B=@n13), B=Multiply(A=@n14, B=@n15))) {name=Rotated Scale, nodeColor=<rgba 0 0 0 1>}
  n16 = MaterialFunctionCall!(Dimensions=ViewSize()) {materialFunction=/FortUILibrary/MaterialFunctions/Basics/MF_UI_GetAspectRatio}
  NamedRerouteDeclaration(None=@n16.Max Ratio) {name=Max Ratio, nodeColor=<rgba 0 0 0 1>}
  n17 = CollectionParameter() {
   collection=&MPC_UI_Master_VaultWorld_Settings
   parameterName=Transition Loading
  }
  n18 = NamedRerouteDeclaration(None=Saturate(None=@n17)) {name=Loading Base Value, nodeColor=<rgba 0 0 0 1>}
  n19 = NamedRerouteUsage()
  n20 = OneMinus(None=NamedRerouteDeclaration(None=Max(A=@n18, B=@n19)) {name=Loading Value, nodeColor=<rgba 0 0 0 1>})
  n21 = NamedRerouteDeclaration(None=ComponentMask(None=NamedRerouteUsage()) {g=true}) {name=Base Gradient, nodeColor=<rgba 0 0 0 1>}
  NamedRerouteDeclaration(None=Ceil(None=Subtract(A=@n21, B=@n20))) {name=Loading Mask, nodeColor=<rgba 0 0 0 1>}
  NamedRerouteDeclaration(None=ScalarParameter() {group=Transition, parameterName=Invert, sliderMax=1}) {name=Invert, nodeColor=<rgba 0 0 0 1>}
  NamedRerouteDeclaration(None=@n16.Min Ratio) {name=Min Ratio, nodeColor=<rgba 0 0 0 1>}
  n22 = NamedRerouteUsage()
  n23 = CollectionParameter() {
   collection=&MPC_UI_Master_VaultWorld_Settings
   parameterName=Transition Reveal
  }
  n24 = LinearInterpolate(A=Multiply(A=@n22) {constB=-0.5}, B=Add(A=Multiply(A=@n22) {constB=0.5}), Alpha=Saturate(None=@n23))
  n25 = Subtract(A=@n21, B=OneMinus(None=NamedRerouteDeclaration(None=@n24) {name=Reveal Value, nodeColor=<rgba 0 0 0 1>}))
  n26 = Saturate(None=NamedRerouteDeclaration(None=Ceil(None=@n25)) {name=Reveal Mask, nodeColor=<rgba 0 0 0 1>}) {desc=B}
  NamedRerouteDeclaration(None=LinearInterpolate(A=OneMinus(None=@n26) {desc=A}, B=@n26, Alpha=NamedRerouteUsage())) {name=Transition Reveal Mask, nodeColor=<rgba 0 0 0 1>}
  n27 = Saturate(None=NamedRerouteUsage()) {desc=A}
  n28 = OneMinus(None=@n27) {desc=B}
  n29 = NamedRerouteUsage()
  NamedRerouteDeclaration(None=Multiply(A=LinearInterpolate(A=@n27, B=@n28, Alpha=@n29), B=NamedRerouteUsage())) {name=Transition Loading Mask, nodeColor=<rgba 0 0 0 1>}
  n30 = Subtract(A=@n9) {constB=0.5}
  n31 = NamedRerouteUsage()
  NamedRerouteDeclaration(None=Add(A=Divide(A=@n30, B=@n31)) {constB=0.5}) {name=Gradient UV, nodeColor=<rgba 0 0 0 1>}
  n32 = NamedRerouteUsage()
  n33 = Add(A=NamedRerouteUsage()) {constB=-0.5}
  n34 = Multiply(A=Subtract(A=Add(A=@n32, B=@n33)) {constB=0.5}, B=NamedRerouteUsage())
  n35 = ComponentMask(None=NamedRerouteDeclaration(None=Add(A=@n34) {constB=0.5}) {name=Divider UV, nodeColor=<rgba 0 0 0 1>}) {g=true}
  n36 = Multiply(A=Min(A=@n35, B=OneMinus(None=@n35))) {constB=2}
  n37 = OneMinus(None=NamedRerouteDeclaration(None=Constant() {r=0.5}) {name=Divider Width, nodeColor=<rgba 0 0 0 1>})
  n38 = Constant() {r=1}
  NamedRerouteDeclaration(None=Saturate(None=MaterialFunctionCall(Input=@n36, Input Low=@n37, Input High=@n38) {materialFunction=&RemapValue_0-1}.Result)) {name=Divider Mask, nodeColor=<rgba 0 0 0 1>}
  StaticBoolParameter() {group=Debug, parameterName=Use New Floor Settings}
 ]
}
